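# Physics asset for the Wicked Duck backpack cosmetic, giving its left and right wings physics bodies for secondary motion.
# Each wing is a mirrored chain of capsules running from the pack body through a base and four joints, with a secondary tip branching off the third joint.
# All constraints lock twist, and those past the wing base also lock one swing axis, so each wing bends like a series of hinges.
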PhysicsAsset /BRCosmetics/Accessories/FORT_Backpacks/Backpack_M_MED_Wicked_Duck/Meshes/M_MED_Wicked_Duck_Pack_Physics
bodies {
 dyn_L_wing_base_jnt [{shapeType=Capsule, length=8, radius=3}]
 dyn_L_wing_jnt_1 [{shapeType=Capsule, length=2, radius=3}]
 dyn_L_wing_jnt_2 [{shapeType=Capsule, length=8, radius=3, center=<0.6082 0 -7.4754>, rotation=<rot 6.1412 0 0>}]
 dyn_L_wing_jnt_3 [{shapeType=Capsule, length=20, radius=2, center=<-0.0284 3.7673 -13.8702>, rotation=<rot -0.614 0 -14.416>}]
 dyn_L_wing_jnt_4 [{shapeType=Capsule, length=30, radius=4, center=<15.3537 5.4985 -5.2515>, rotation=<rot 5.6377 103.8466 81.8054>}]
 dyn_L_wing_jnt_secondary_4 [{shapeType=Capsule, length=8, radius=1, center=<6.3855 -0.0001 -3.6508>, rotation=<rot -75.2733 -179.9998 179.9998>}]
 dyn_R_wing_base_jnt [{shapeType=Capsule, length=8, radius=3}]
 dyn_R_wing_jnt_1 [{shapeType=Capsule, length=2, radius=3}]
 dyn_R_wing_jnt_2 [{shapeType=Capsule, length=8, radius=3, center=<0.6082 0 -7.4754>, rotation=<rot 6.1412 0 0>}]
 dyn_R_wing_jnt_3 [{shapeType=Capsule, length=20, radius=2, center=<-0.0188 -1.8507 -12.984>, rotation=<rot -0.614 0.0002 13.6447>}]
 dyn_R_wing_jnt_4 [{shapeType=Capsule, length=30, radius=4, center=<14.4477 -3.6522 -6.1096>, rotation=<rot -5.4058 76.7358 79.8587>}]
 dyn_R_wing_jnt_secondary_4 [{shapeType=Capsule, length=8, radius=1, center=<6.3853 -0.0001 -3.651>, rotation=<rot -75.271 -179.9999 179.9999>}]
 pack [{shapeType=Capsule, length=8, radius=5, center=<0 0 2.0734>, rotation=<rot 0 90 90>}]
}
constraints [
 {bone1Name=dyn_L_wing_base_jnt, bone2Name=pack, swing1LimitDegrees=10, swing2LimitDegrees=5, twistLimitDegrees=45, twistMotion=Locked}
 {
  bone1Name=dyn_L_wing_jnt_1
  bone2Name=dyn_L_wing_base_jnt
  swing1LimitDegrees=45
  swing1Motion=Locked
  swing2LimitDegrees=5
  twistLimitDegrees=45
  twistMotion=Locked
 }
 {
  bone1Name=dyn_L_wing_jnt_2
  bone2Name=dyn_L_wing_jnt_1
  swing1LimitDegrees=45
  swing1Motion=Locked
  swing2LimitDegrees=5
  twistLimitDegrees=45
  twistMotion=Locked
 }
 {
  bone1Name=dyn_L_wing_jnt_3
  bone2Name=dyn_L_wing_jnt_2
  swing1LimitDegrees=45
  swing1Motion=Locked
  swing2LimitDegrees=5
  twistLimitDegrees=45
  twistMotion=Locked
 }
 {
  bone1Name=dyn_L_wing_jnt_4
  bone2Name=dyn_L_wing_jnt_3
  swing1LimitDegrees=5
  swing1Motion=Locked
  swing2LimitDegrees=15
  twistLimitDegrees=45
  twistMotion=Locked
 }
 {bone1Name=dyn_R_wing_base_jnt, bone2Name=pack, swing1LimitDegrees=10, swing2LimitDegrees=5, twistLimitDegrees=45, twistMotion=Locked}
 {
  bone1Name=dyn_R_wing_jnt_1
  bone2Name=dyn_R_wing_base_jnt
  swing1LimitDegrees=45
  swing1Motion=Locked
  swing2LimitDegrees=5
  twistLimitDegrees=45
  twistMotion=Locked
 }
 {
  bone1Name=dyn_R_wing_jnt_2
  bone2Name=dyn_R_wing_jnt_1
  swing1LimitDegrees=45
  swing1Motion=Locked
  swing2LimitDegrees=5
  twistLimitDegrees=45
  twistMotion=Locked
 }
 {
  bone1Name=dyn_R_wing_jnt_3
  bone2Name=dyn_R_wing_jnt_2
  swing1LimitDegrees=45
  swing1Motion=Locked
  swing2LimitDegrees=5
  twistLimitDegrees=45
  twistMotion=Locked
 }
 {
  bone1Name=dyn_R_wing_jnt_4
  bone2Name=dyn_R_wing_jnt_3
  swing1LimitDegrees=5
  swing1Motion=Locked
  swing2LimitDegrees=15
  twistLimitDegrees=45
  twistMotion=Locked
 }
 {
  bone1Name=dyn_L_wing_jnt_secondary_4
  bone2Name=dyn_L_wing_jnt_3
  swing1LimitDegrees=3
  swing1Motion=Locked
  swing2LimitDegrees=8
  twistLimitDegrees=45
  twistMotion=Locked
 }
 {
  bone1Name=dyn_R_wing_jnt_secondary_4
  bone2Name=dyn_R_wing_jnt_3
  swing1LimitDegrees=3
  swing1Motion=Locked
  swing2LimitDegrees=8
  twistLimitDegrees=45
  twistMotion=Locked
 }
]
