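# Animation Blueprint for the Wicked Duck backpack, blending between skydiving and grounded poses with physics-driven motion.
# In flight, a state machine plays a movement blendspace or floating animation, softened by a partial-alpha RigidBody node.
# Otherwise, RigidBody simulates the input pose with ground-dependent gravity, then both wing base bones rotate while crouching or targeting.
# The event graph caches the pawn and updates the blendspace's directional inputs from player movement.
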AnimBlueprint /BRCosmetics/Accessories/FORT_Backpacks/Backpack_M_MED_Wicked_Duck/Meshes/M_MED_Wicked_Duck_Pack_AnimBP
target_skeleton=&/M_MED_Wicked_Duck_Pack_Skeleton
graphs [
 AnimGraph {nodes [
   n1 = RigidBody!(
    ComponentPose=Local To Component(LocalPose=Use cached pose 'FlightPosesCache'())
    Alpha=0.4
    OverrideWorldGravity:Vector="0,0,-500"
    ExternalForce
   ) -> (Pose) {
    bClampLinearTranslationLimitToRefPose=true
    bOverrideWorldGravity=true
    overridePhysicsAsset=&/M_MED_Wicked_Duck_Pack_Physics
    simulationSpace=BaseBoneSpace
    alphaBoolBlend {blendOption=Linear}
    alphaScaleBiasClamp {bInterpResult=true, interpSpeedDecreasing=30, interpSpeedIncreasing=30}
    baseBoneRef {boneName=root}
    componentAppliedLinearAccClamp=<5000 5000 5000>
    componentLinearAccScale=<0.25 0.25 0.25>
    componentLinearVelScale=<0.4 0.4 0.4>
   }
   Output Pose(Result=Blend Poses by bool(BlendPose_0=Component To Local(ComponentPose=@n1), BlendPose_1=Use cached pose 'Grounded'(), bActiveValue=Get bIsSkyDiving()) {blendType=Linear}) {layerGroup=None}
   AnimGraphNode_StateMachine_0 = Flight_Poses() {is=title AnimGraphNode_StateMachine}
   Save cached pose 'FlightPosesCache'(Pose=@AnimGraphNode_StateMachine_0) {cacheName=FlightPosesCache}
   n2 = RigidBody(
    ComponentPose=Local To Component(LocalPose=Input Pose())
    OverrideWorldGravity=SelectVector(A="0, 0,-1000", B="0,0,-750", bPickA=Get bIsOnGround() -> (bIsOnGround:bool))
   ) {
    bClampLinearTranslationLimitToRefPose=true
    bOverrideWorldGravity=true
    overridePhysicsAsset=&M_MED_Wicked_Duck_Pack_Physics
    simulationSpace=BaseBoneSpace
    alphaBoolBlend {blendOption=Linear}
    alphaScaleBiasClamp {bInterpResult=true, interpSpeedDecreasing=30, interpSpeedIncreasing=30}
    baseBoneRef {boneName=root}
    componentAppliedLinearAccClamp=<5000 5000 5000>
    componentLinearAccScale=<0.25 0.25 0.25>
    componentLinearVelScale=<0.4 0.4 0.4>
   }
   n3 = "To Float (Boolean)"(InBool=OR Boolean!(A=Get bIsCrouching(), B=Get bIsCrouchSprinting(), C:bool=Get bIsTargeting()) -> (ReturnValue))
   n4 = "Transform (Modify) Bone - Bone: dyn_L_wing_base_jnt"(ComponentPose=@n2, Alpha=@n3, Translation="0,0,0", Rotation="0,-20,0", Scale="1,1,1") {
    rotationMode=BMM_Replace
    rotationSpace=BCS_BoneSpace
    alphaBoolBlend {blendOption=Linear}
    alphaScaleBiasClamp {bInterpResult=true, interpSpeedDecreasing=15, interpSpeedIncreasing=15}
    boneToModify {boneName=dyn_L_wing_base_jnt}
    rotation {yaw=-20}
   }
   n5 = "Transform (Modify) Bone - Bone: dyn_R_wing_base_jnt"(ComponentPose=@n4, Alpha=@n3, Translation="0,0,0", Rotation="0,20,0", Scale="1,1,1") {
    rotationMode=BMM_Replace
    rotationSpace=BCS_BoneSpace
    alphaBoolBlend {blendOption=Linear}
    alphaScaleBiasClamp {bInterpResult=true, interpSpeedDecreasing=15, interpSpeedIncreasing=15}
    boneToModify {boneName=dyn_R_wing_base_jnt}
    rotation {yaw=20}
   }
   Save cached pose 'Grounded'(Pose=Component To Local(ComponentPose=@n5)) {cacheName=Grounded}
  ]}
 Flight_Poses {
  entry=Base
  path=AnimGraph.AnimGraphNode_StateMachine_0.Flight_Poses
  ids=id
  nodes [
   AnimStateNode_0 = Base() {is=title AnimStateNode}
   AnimStateNode_1 = Skydiving() {
    is=title AnimStateNode
    stateEntered {guid=#1}
    stateFullyBlended {guid=#2}
    stateLeft {guid=#3}
   }
   AnimStateTransitionNode_0 = AnimStateTransitionNode()
   AnimStateTransitionNode_1 = AnimStateTransitionNode()
  ]
  links [
   {
    dst=Skydiving
    rule=AnimGraph.AnimGraphNode_StateMachine_0.Flight_Poses.AnimStateTransitionNode_0.Transition
    src=Base
    via=AnimStateTransitionNode_0
   }
   {
    dst=Base
    rule=AnimGraph.AnimGraphNode_StateMachine_0.Flight_Poses.AnimStateTransitionNode_1.Transition
    src=Skydiving
    via=AnimStateTransitionNode_1
   }
  ]
 }
 Base {
  path=AnimGraph.AnimGraphNode_StateMachine_0.Flight_Poses.AnimStateNode_0.Base
  nodes [Output Animation Pose(Result=Sequence Player 'Air_Base'() {sequence=/BRCosmetics/Accessories/FORT_Backpacks/Backpack_M_MED_Wicked_Duck/Animations/Air_Base})]
 }
 Skydiving {
  path=AnimGraph.AnimGraphNode_StateMachine_0.Flight_Poses.AnimStateNode_1.Skydiving
  nodes [
   n1 = Get LeftRight() -> (LeftRight:double)
   n2 = Get ForwardBack() -> (ForwardBack:double)
   Output Animation Pose(Result=Blend Poses by bool(
     BlendPose_0=Blendspace Player 'M_MED_Wicked_Duck_Airborne_BS'(X=@n1, Y=@n2) {blendSpace=/BRCosmetics/Accessories/FORT_Backpacks/Backpack_M_MED_Wicked_Duck/Animations/M_MED_Wicked_Duck_Airborne_BS}
     BlendPose_1=Sequence Player 'Air_Floating'() {sequence=/BRCosmetics/Accessories/FORT_Backpacks/Backpack_M_MED_Wicked_Duck/Animations/Air_Floating}
     BlendTime_0=0.3
     BlendTime_1=0.3
     bActiveValue=Get bIsAccelerating2D()
    ) {blendType=Linear})
  ]
 }
 Transition {
  path=AnimGraph.AnimGraphNode_StateMachine_0.Flight_Poses.AnimStateTransitionNode_0.Transition
  nodes [
   Result(bCanEnterTransition=Get bIsSkyDiving())
   OR Boolean!(A=false, B=Get bIsSkydiveFloating()) -> (ReturnValue)
  ]
 }
 Transition {
  path=AnimGraph.AnimGraphNode_StateMachine_0.Flight_Poses.AnimStateTransitionNode_1.Transition
  nodes [
   Result(bCanEnterTransition=NOT Boolean(A=Get bIsSkyDiving()))
   Get bIsSkydiveFloating()
  ]
 }
 EventGraph {nodes [
   Event BlueprintInitializeAnimation()
   > Set SkelMesh(SkelMesh:Skeletal Mesh Object Reference=&/M_MED_Wicked_Duck_Pack) -> (Output_Get:Skeletal Mesh Object Reference)
   > n1 = Cast To FortPlayerPawn(Object=TryGetFortPawn())
   > Set Pawn(Pawn:Fort Player Pawn Object Reference=@n1) -> (Output_Get:Fort Player Pawn Object Reference)
   Event BlueprintUpdateAnimation()
   > Get!(execute:Exec) -> (then:Exec, else:Exec, SkelMesh:Skeletal Mesh Object Reference)
   > Get!(execute) -> (then, else, Pawn:Fort Player Pawn Object Reference)
   > n2 = PlayerMovement(AnimInstance=Self-Reference(), XDotProductCoefficient=25, YDotProductCoefficient=25, ZDotProductCoefficient=1)
   > n3 = Sequence()
   > Set LeftRight(LeftRight:double=@n2) -> (Output_Get:double)
   Set ForwardBack(execute=@n3.then_1, ForwardBack:double=@n2.YMovement) -> (Output_Get:double)
  ]}
]
variables [
 ForwardBack {type=number}
 LeftRight {type=number}
 Pawn {type=FortPlayerPawn}
 SkelMesh {type=SkeletalMesh, default=null}
]
